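; Gameplay ability that spawns clones for the Magic Donut items and tags each clone according to the item that triggered it.
; On activation it matches the source item's display-name text, Explode or Recall, to pick the T1 or T2 tag set, then raises a ready flag even without a match.
; The OnCloneActorSpawned override adds those tags as replicated loose gameplay tags once that flag is set, leaving spawning to the parent class.
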(event Ability|EventActivateAbility
  (|CallParentFunction)
  (switch string (Utilities|String|ToString(Text) (ItemDefinition|GetDisplayName (Ability|GetCurrentSourceItemDefinition)))
    (:Magic%20Donut%20Explode
      (Variables|Default|SetGameplay%20Tags (Variables|Default|GetT1%20Tags)))
    (:Magic%20Donut%20Recall
      (Variables|Default|SetGameplay%20Tags (Variables|Default|GetT2%20Tags))))
  (Variables|Default|SetGameplayTagsSet true))

(fn OnCloneActorSpawned
  (params (EncounterInstance :type Fort%20Athena%20Living%20World%20Encounter%20Instance%20Object%20Reference) (SpawnedActor :type Actor%20Object%20Reference) (EventData :type Fort%20Athena%20Living%20World%20Event%20Data%20Object%20Reference))
  (|CallParentFunction EncounterInstance SpawnedActor EventData)
  (if (Variables|Default|GetGameplayTagsSet)
    (Ability|GameplayEffect|AddLooseGameplayTags SpawnedActor (Variables|Default|GetGameplay%20Tags) true)))
