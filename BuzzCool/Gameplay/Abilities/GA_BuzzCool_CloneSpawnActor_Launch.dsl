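; Gameplay ability that launches a BuzzCool clone pawn away from its spawn provider in a randomized direction.
; On activation it caches the avatar as the clone, commits, computes force data, and applies a constant-force root motion task for a scalable duration.
; The provider's forward vector gets random yaw and pitch around world Z and Y, with random strength, all from ScalableFloat ranges.
; No EndAbility call appears in the graph.
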(event Ability|EventActivateAbility
  (Development|PrintString "@@@ Event ActivateAbility [GA_BuzzCool_CloneSpawnActor]: CALLED!" false (Variables|Debug|GetbDebugAbility))
  (CallFunction|InitAbility)
  (Ability|CommitAbility)
  (CallFunction|UpdateForceData)
  (Ability|Tasks|ApplyRootMotionConstantForce "None" (Variables|Internal|GetForceDirection) (Variables|Internal|GetForceStrength) (ScalableFloat|GetValueatLevel (Variables|Internal|GetSF_LaunchDuration))))

(event Ability|EventOnEndAbility
  (params (bWasCancelled :type Boolean))
  (Development|PrintString "@@@ Event OnEndAbility [GA_BuzzCool_CloneSpawnActor]: CALLED!" false (Variables|Debug|GetbDebugAbility)))

(fn InitAbility
  (bind _returnvalue (Ability|GetAvatarActorfromActorInfo))
  (Utilities|IsValid _returnvalue
    (:Is%20Valid
      (bind _aspawn_wild_estate_buzz_cool (Utilities|Casting|CastToPawn_WildEstate_BuzzCool _returnvalue))
      (Variables|Internal|SetCloneActor _aspawn_wild_estate_buzz_cool))))

(fn UpdateForceData
  (bind _cloneactor (Variables|Internal|GetCloneActor))
  (bind _spawnproviderref (Variables|/BuzzCool/NPCs/BuzzCool_Clones/Pawn/Pawn_WildEstate_BuzzCool.Pawn_WildEstate_BuzzCool_C|GetSpawnProviderRef _cloneactor))
  (bind (_returnvalue_location _returnvalue_rotation) (Math|Transform|BreakTransform (Transformation|GetActorTransform _spawnproviderref)))
  (Variables|ForceData|SetForceBaseLocation _returnvalue_location)
  (bind _output_get (Variables|ForceData|SetForceBaseRotation _returnvalue_rotation))
  (LocalVariables|SetCurrentBaseForce _output_get)
  (Variables|ForceData|SetForceAdjustedVector (Math|Vector|RotateVectorAroundAxis (Math|Vector|RotateVectorAroundAxis (Math|Vector|GetForwardVector (LocalVariables|GetCurrentBaseForce)) (Math|Random|RandomFloatinRange (ScalableFloat|GetValueatLevel (Variables|Internal|GetSF_LaunchYawMin)) (ScalableFloat|GetValueatLevel (Variables|Internal|GetSF_LaunchYawMax))) "0.000000,0.000000,1.000000") (Math|Random|RandomFloatinRange (ScalableFloat|GetValueatLevel (Variables|Internal|GetSF_LaunchPitchMin)) (ScalableFloat|GetValueatLevel (Variables|Internal|GetSF_LaunchPitchMax))) "0.000000,1.000000,0.000000"))
  (if (Variables|Debug|GetbDebugAbility)
    (Rendering|Debug|DrawDebugLine (Variables|ForceData|GetForceBaseLocation) (+ (Variables|ForceData|GetForceBaseLocation) (* (Math|Vector|GetForwardVector (Variables|ForceData|GetForceBaseRotation)) 5e+02)) "(R=0.000000,G=0.484375,B=0.284624,A=1.000000)" 1e+01 3.0)
    (Rendering|Debug|DrawDebugLine (Variables|ForceData|GetForceBaseLocation) (+ (Variables|ForceData|GetForceBaseLocation) (* (Variables|ForceData|GetForceAdjustedVector) 5e+02)) "(R=1.000000,G=0.044101,B=0.781620,A=1.000000)" 1e+01 3.0))
  (Variables|Internal|SetForceDirection (Variables|ForceData|GetForceAdjustedVector))
  (Variables|Internal|SetForceStrength (Math|Random|RandomFloatinRange (ScalableFloat|GetValueatLevel (Variables|Internal|GetSF_LaunchStrengthMin)) (ScalableFloat|GetValueatLevel (Variables|Internal|GetSF_LaunchStrengthMax)))))
