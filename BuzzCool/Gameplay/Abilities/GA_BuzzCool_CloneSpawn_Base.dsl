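; Base gameplay ability that spawns BuzzCool AI clones allied to the activating player through a Living World encounter, running on authority.
; It attaches a volume to the player or override actor, then starts or resumes the encounter once registered.
; Spawned clones join the player's team, follow them unless overridden, and turn hostile toward their attackers.
; Ending pauses the encounter; a scalable flag also ends it, destroying the volume and killing clones.
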(event Ability|EventActivateAbility
  (Ability|HasAuthority
    (:True
      (Development|PrintString "@@@ Event ActivateAbility [GA_BuzzCool_CloneSpawn_Base]: Authority - CALLED!" false (Variables|Debug|GetbDebugAbility))
      (Utilities|FlowControl|Sequence
        (:then_0
          (CallFunction|InitializeAbilityValues))
        (:then_1
          (Utilities|IsValid (Variables|Internal|GetLivingWorldVol_BuzzCool)
            (:Is%20Valid
              (CallFunction|AttachPointProvider))
            (:Is%20Not%20Valid
              (CallFunction|IntializePointProvider))))))))

(event Custom|IntializePointProvider
  (Development|PrintString "@@@@@@ InitializePointProvider [GA_BuzzCool_CloneSpawn_Base]: CALLED!" false (Variables|Debug|GetbDebugAbility))
  (bind (_asynctaskproxy _spawnedactor) (Ability|Tasks|SpawnActor :Class (Variables|Internal|GetLivingWorldVol_Class) :Instigator (Variables|Internal|GetInstigatingPlayerPawn))
    (:then
      (bind _asfort_athena_living_world_volume (Utilities|Casting|CastToFortAthenaLivingWorldVolume _spawnedactor)
        (:then
          (Variables|Internal|SetLivingWorldVol_BuzzCool _asfort_athena_living_world_volume)
          (CallFunction|AttachPointProvider))
        (:CastFailed
          (Fortnite|FortShippingLog (Variables|Internal|GetInstigatingPlayerPawn) "@@@@@@ InitializePointProvider [GA_BuzzCool_CloneSpawn_Base]: ERROR - created point provider is not of the expected kind!" true)
          (Ability|EndAbility))))
    (:DidNotSpawn
      (Fortnite|FortShippingLog (Variables|Internal|GetInstigatingPlayerPawn) "@@@@@@ InitializePointProvider [GA_BuzzCool_CloneSpawn_Base]: ERROR - failed to create a valid point provider!" true)
      (Ability|EndAbility))))

(event Custom|AttachPointProvider
  (bind _livingworldvol_buzzcool (Variables|Internal|GetLivingWorldVol_BuzzCool))
  (Development|PrintString "@@@@@@ AttachPointProvider [GA_BuzzCool_CloneSpawn_Base]: CALLED!" false (Variables|Debug|GetbDebugAbility))
  (Transformation|AttachActorToActor _livingworldvol_buzzcool (select (and (Utilities|IsValid (Variables|Tunables|GetLivingWorldVol_AttachActor) :!class /Script/Engine.KismetSystemLibrary) (Variables|Tunables|GetbUseAttachOverride)) (Variables|Tunables|GetLivingWorldVol_AttachActor) (Variables|Internal|GetInstigatingPlayerPawn)) "None" "KeepRelative" "KeepRelative" "KeepRelative" true)
  (EventDispatchers|BindEventtoOnPointProviderRegistered _livingworldvol_buzzcool (delegate Custom|OnLivingWorldVolumeReady))
  (Class|FortAthenaLivingWorldPointProviderInterface|EnablePointProvider_DEPRECATED _livingworldvol_buzzcool))

(event Custom|OnLivingWorldVolumeReady
  (params (PointProviderVolume :type Fort%20Athena%20Living%20World%20Volume%20Object%20Reference))
  (Class|FortAthenaLivingWorldManager|RequestEventGeneration (Class|FortAthenaLivingWorldManager|GetLivingWorldManager))
  (CallFunction|StartCloneEncounter))

(event Ability|EventOnEndAbility
  (params (bWasCancelled :type Boolean))
  (Ability|HasAuthority
    (:True
      (Development|PrintString "@@@ Event OnEndAbility [GA_BuzzCool_CloneSpawn_Base]: Authority - CALLED!" false (Variables|Debug|GetbDebugAbility))
      (EventDispatchers|UnbindallEventsfromOnDamaged (Variables|Internal|GetInstigatingPlayerPawn))
      (CallFunction|StopCloneEncounter))))

(fn AddHostileTarget
  (params (DamagedActor :type Actor%20Object%20Reference) (Damage :type Float%20%28single-precision%29) (InstigatedBy :type Controller%20Object%20Reference) (DamageCauser :type Actor%20Object%20Reference) (HitLocation :type Vector) (FHitComponent :type Primitive%20Component%20Object%20Reference) (BoneName :type Name) (Momentum :type Vector))
  (Development|PrintString "@@@@@@@@@ AddHostileTarget [GA_BuzzCool_CloneSpawn_Base]: CALLED!" false (Variables|Debug|GetbDebugAbility))
  (for _array_element (Variables|Internal|GetCloneActorArray)
    (Utilities|IsValid _array_element
      (:Is%20Valid
        (bind _asfort_aipawn (Utilities|Casting|CastToFortAIPawn _array_element))
        (bind _asfort_player_pawn (Utilities|Casting|CastToFortPlayerPawn (Pawn|GetControlledPawn InstigatedBy)))
        (Affiliation|ChangeAttitudeBetweenActors _asfort_aipawn _asfort_player_pawn "Hostile")))))

(fn AddUniqueCloneActor
  (params (Actor :type Actor%20Object%20Reference))
  (bind _instigatingplayerpawn (Variables|Internal|GetInstigatingPlayerPawn))
  (bind _cloneactorarray (Variables|Internal|GetCloneActorArray))
  (Utilities|IsValid Actor
    (:Is%20Valid
      (if (not (Utilities|Array|ContainsItem _cloneactorarray Actor))
        (Utilities|Array|AddUnique _cloneactorarray Actor)
        (Affiliation|ChangeAttitudeBetweenActors Actor _instigatingplayerpawn)
        (Teams|ChangeTeam Actor 0 (Team|GetActorTeam _instigatingplayerpawn))))))

(fn InitializeAbilityValues
  (Development|PrintString "@@@@@@ InitializeAbilityValues [GA_BuzzCool_CloneSpawn_Base]: CALLED!" false (Variables|Debug|GetbDebugAbility))
  (Utilities|Array|Clear (Variables|Internal|GetCloneActorArray))
  (bind _outfortplayerpawn (CallFunction|GetAvatarasValidFortPlayerPawn)
    (:ValidFortPlayerPawn
      (bind _output_get (Variables|Internal|SetInstigatingPlayerPawn _outfortplayerpawn))
      (EventDispatchers|BindEventtoOnDamaged _output_get (EventDispatchers|CreateEvent :!event AddHostileTarget)))
    (:AvatarCastFailed
      (Ability|EndAbility))))

(fn OnCloneActorDied
  (params (EncounterInstance :type Fort%20Athena%20Living%20World%20Encounter%20Instance%20Object%20Reference) (DiedActor :type Actor%20Object%20Reference))
  (Development|PrintString "@@@@@@@@@ OnCloneActorDied [GA_BuzzCool_CloneSpawn_Base]: CALLED!" false (Variables|Debug|GetbDebugAbility))
  (Utilities|Casting|CastToFortAIPawn DiedActor)
  (Utilities|Array|RemoveItem (Variables|Internal|GetCloneActorArray) DiedActor))

(fn OnCloneActorSpawned
  (params (EncounterInstance :type Fort%20Athena%20Living%20World%20Encounter%20Instance%20Object%20Reference) (SpawnedActor :type Actor%20Object%20Reference) (EventData :type Fort%20Athena%20Living%20World%20Event%20Data%20Object%20Reference))
  (Development|PrintString "@@@@@@@@@ OnCloneActorSpawned [GA_BuzzCool_CloneSpawn_Base]: CALLED!" false (Variables|Debug|GetbDebugAbility)))

(fn OnCloneEventSpawned
  (params (EncounterInstance :type Fort%20Athena%20Living%20World%20Encounter%20Instance%20Object%20Reference) (SpawnedActors :type Array%20of%20Actor%20Object%20References))
  (Development|PrintString "@@@@@@@@@ OnCloneEventSpawned [GA_BuzzCool_CloneSpawn_Base]: CALLED!" false (Variables|Debug|GetbDebugAbility))
  (for _array_element SpawnedActors
    (bind _asfort_aipawn (Utilities|Casting|CastToFortAIPawn _array_element))
    (CallFunction|AddUniqueCloneActor _asfort_aipawn)
    (bind _asaicontroller_buzz_cool_clone (Utilities|Casting|CastToAIController_BuzzCool_Clone (Pawn|GetController _asfort_aipawn)))
    (if (Variables|Tunables|GetbUseAttachOverride)
      (Class|AIControllerBuzzCoolClone|StopPawnActorPolling _asaicontroller_buzz_cool_clone)
      (Class|AIControllerBuzzCoolClone|UpdateCloneLeash _asaicontroller_buzz_cool_clone 0)
      (else
        (Variables|/BuzzCool/NPCs/BuzzCool_Clones/Controllers/AIController_BuzzCool_Clone.AIController_BuzzCool_Clone_C|SetCurrentGoalPawnActor (Variables|Internal|GetInstigatingPlayerPawn) _asaicontroller_buzz_cool_clone)
        (Class|AIControllerBuzzCoolClone|StartPawnActorPolling _asaicontroller_buzz_cool_clone))))
  (return))

(fn StartCloneEncounter
  (bind _cloneencounterinstance (Variables|Internal|GetCloneEncounterInstance))
  (Development|PrintString "@@@@@@ StartCloneEncounter [GA_BuzzCool_CloneSpawn_Base]: CALLED!" false (Variables|Debug|GetbDebugAbility))
  (Utilities|IsValid _cloneencounterinstance
    (:Is%20Valid
      (Class|FortAthenaLivingWorldEncounterInstance|ResumeEncounter _cloneencounterinstance))
    (:Is%20Not%20Valid
      (bind _activityencounter (Variables|Tunables|GetActivityEncounter)
        (:then
          (bind (_isvalid _returnvalue) (Default|StartActivityEncounter (Fortnite|Athena|GameState|GetGameStateAthena) (Variables|Internal|GetLivingWorldVol_BuzzCool) (Variables|Internal|GetInstigatingPlayerPawn) _activityencounter))
          (bind _output_get (Variables|Internal|SetCloneEncounterInstance _returnvalue))
          (EventDispatchers|BindEventtoOnEventSpawned _output_get (EventDispatchers|CreateEvent :!event OnCloneEventSpawned))
          (EventDispatchers|BindEventtoOnActorDied _output_get (EventDispatchers|CreateEvent :!event OnCloneActorDied))
          (EventDispatchers|BindEventtoOnActorSpawned _output_get (EventDispatchers|CreateEvent :!event OnCloneActorSpawned)))
        (:else
          (Fortnite|FortShippingLog (Variables|Internal|GetInstigatingPlayerPawn) "@@@@@@ StartCloneEncounter [GA_BuzzCool_CloneSpawn_Base]: ERROR - invalid activity encounter declared!" true)
          (Ability|EndAbility))))))

(fn StopCloneEncounter
  (bind _cloneencounterinstance (Variables|Internal|GetCloneEncounterInstance))
  (Development|PrintString "@@@@@@ StopCloneEncounter [GA_BuzzCool_CloneSpawn_Base]: CALLED!" false (Variables|Debug|GetbDebugAbility))
  (Class|FortAthenaLivingWorldEncounterInstance|PauseEncounter (Variables|Internal|GetCloneEncounterInstance))
  (if (ScalableFloat|GetValueasBool (Variables|Tunables|GetSF_EndCloneEncounterWithAbility))
    (Class|BPFLLagerCommunFunctions|EndActivityEncounter (Game|GetGameState) (Variables|Internal|GetCloneEncounterInstance))
    (Actor|DestroyActor (Variables|Internal|GetLivingWorldVol_BuzzCool))
    (EventDispatchers|UnbindEventfromOnEventSpawned _cloneencounterinstance (EventDispatchers|CreateEvent :!event OnCloneEventSpawned))
    (EventDispatchers|UnbindEventfromOnActorDied _cloneencounterinstance (EventDispatchers|CreateEvent :!event OnCloneActorDied))
    (EventDispatchers|UnbindEventfromOnActorSpawned _cloneencounterinstance (EventDispatchers|CreateEvent :!event OnCloneActorSpawned))
    (for _array_element (Variables|Internal|GetCloneActorArray)
      (Utilities|IsValid _array_element
        (:Is%20Valid
          (bind _asfort_pawn (Utilities|Casting|CastToFortPawn _array_element))
          (FortPawn|ForceKill _asfort_pawn "(TagName=\"Gameplay.Effect.InstantDeath.Despawn\")" (AI|GetAIController _asfort_pawn) _asfort_pawn))))))
