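; Gameplay ability for the BuzzCool clone-spawn item, extending a parent ability that is called to continue activation.
; On authority only, it takes the instigator actor from the granting effect's context and sets it as the living world volume attach actor before calling the parent.
; It also binds to that instigator's destroyed event, so the ability ends when the instigator is destroyed.
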(event Ability|EventActivateAbility
  (bind _returnvalue (Ability|GetGrantedbyEffectContext))
  (bind _returnvalue_2 (Ability|EffectContext|GetInstigatorActor _returnvalue))
  (Ability|HasAuthority
    (:True
      (Variables|Tunables|SetLivingWorldVol_AttachActor _returnvalue_2)
      (Game|BindEventtoOnDestroyed _returnvalue_2 (delegate Custom|OnDestroyedInstigator))
      (|CallParentFunction))))

(event Custom|OnDestroyedInstigator
  (params (DestroyedActor :type Actor%20Object%20Reference))
  (Ability|EndAbility))
