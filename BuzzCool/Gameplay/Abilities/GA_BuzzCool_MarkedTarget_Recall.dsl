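; GA_BuzzCool_MarkedTarget_Recall is a gameplay ability that clears the BuzzCool marked-target state from its owner when the mark's instigator gains the Recall tag.
; On activation it watches the instigator of its granting effect for Item.BuzzCool.Recall, then removes owner effects granting the MarkedTarget tag.
; In parallel, it ends itself when that tag leaves the avatar pawn, or immediately if the pawn cast fails.
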(event Ability|EventActivateAbility
  (Utilities|FlowControl|Sequence
    (:then_0
      (bind _outfortplayerpawn (CallFunction|GetAvatarasValidFortPlayerPawn)
        (:ValidFortPlayerPawn
          (Ability|Tasks|WaitGameplayTagRemoveFromActor _outfortplayerpawn "(TagName=\"Ability.BuzzCool.Item.MarkedTarget\")"
            (:Removed
              (Ability|EndAbility))))
        (:AvatarCastFailed
          (Ability|EndAbility))))
    (:then_1
      (Ability|Tasks|WaitGameplayTagAddToActor (Ability|EffectContext|GetInstigatorActor (Ability|GetGrantedbyEffectContext)) "(TagName=\"Item.BuzzCool.Recall\")" true
        (:Added
          (Ability|RemoveGameplayEffectFromOwnerWithGrantedTags "(GameplayTags=((TagName=\"Ability.BuzzCool.Item.MarkedTarget\")))"))))))

(event Ability|EventOnEndAbility
  (params (bWasCancelled :type Boolean)))
