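; GA_BuzzCool_RecallClones is a gameplay ability for the BuzzCool item that, judging by its name, recalls clones.
; On activation it gets the avatar's Fort player pawn and ends immediately if that pawn already has the Item.BuzzCool.Recall tag.
; Otherwise it applies GE_BuzzCool_RecallClones to itself, removes the WildEstate BuzzCool cooldown effect, starts a montage task, then ends.
; The montage's completed and cancelled outputs and the end-ability event are empty.
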(event Ability|EventActivateAbility
  (bind _outfortplayerpawn (CallFunction|GetAvatarasValidFortPlayerPawn)
    (:ValidFortPlayerPawn
      (if (GameplayTags|HasMatchingGameplayTag (Ability|GetFortAbilitySystemComponent _outfortplayerpawn) "(TagName=\"Item.BuzzCool.Recall\")")
        (Ability|EndAbility)
        (else
          (GameplayEffects|ApplyGameplayEffectToSelf (Ability|GetFortAbilitySystemComponent _outfortplayerpawn) "/BuzzCool/Gameplay/Abilities/GE_BuzzCool_RecallClones.GE_BuzzCool_RecallClones_C" 0.0 (Class|AthenaFunctionLibrary|MakeSimpleFortEffectContext _outfortplayerpawn))
          (GameplayEffects|RemoveActiveEffectswithGrantedTags (Ability|GetFortAbilitySystemComponent _outfortplayerpawn) "(GameplayTags=((TagName=\"Granted.Cooldown.WildEstate.BuzzCool\")))")
          (Ability|Tasks|PlayMontageInfoAndWaitForTrigger "None" (Variables|Animation|GetMontageInfo)
            (:Cancelled
              _)
            (:Completed
              _))
          (Ability|EndAbility))))))

(event Ability|EventOnEndAbility
  (params (bWasCancelled :type Boolean)))
